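# Niagara effect for the ForwardLake Dire emote that pairs skeletal-mesh motion trail ribbons with rising steam puffs.
# MotionTrail001 samples mesh sockets after a short loop delay, with emitter-age curves shaping spawning, width and fade across two ribbon renderers.
# Steam emits GPU flipbook sprites from a small sphere, using system age to drive upward velocity, color and late fade.
# Time-of-day scalars adjust both emitters' opacity; steam also varies by front-end context.
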NiagaraSystem /BRCosmetics/Animation/Game/MainPlayer/Emotes/ForwardLake_Dire/FX/NS_Emote_ForwardLakeDire_MotionTrailSteam
dependencies {
 usedDataInterfaces [
  &NiagaraDataInterfaceCurve
  &NiagaraDataInterfaceVectorCurve
  &NiagaraDataInterfaceVector2DCurve
  &NiagaraDataInterfaceSpriteRendererInfo
  &NiagaraDataInterfaceColorCurve
  &NiagaraDataInterfaceArrayInt32
  &NiagaraDataInterfaceArrayFloat
 ]
 usedDynamicInputs [
  &Vector2DFromCurve
  &RandomRangeVector2D
  &SelectFloatFromArray
 ]
 usedModules [
  &SystemState
  &self:SystemSpawnScript.SystemScriptSource.SystemScriptGraph.NiagaraNodeAssignment_1.SetVariables_#1
  &self:SystemSpawnScript.SystemScriptSource.SystemScriptGraph.NiagaraNodeAssignment_0.SetVariables_#2
 ]
}
emitters {
 MotionTrail001 {
  emitterUpdateScript [
   &EmitterState(
    Life Cycle Mode=::NewEnumerator1
    Loop Behavior=::NewEnumerator1
    Loop Delay=0.287
    Loop Duration=1
    Scalability Mode=::NewEnumerator1
    UseLoopDelay=true
   )
   &SpawnRate(
    SpawnGroup=1
    SpawnRate=&FloatFromCurve(FloatCurve=DataInterface() {propertyValues {Curve {keys [<at 0.27 1>,<at 0.3 0>]}}}, CurveIndex=$Emitter.NormalizedLoopAge, Scale Curve=15)
    Use Spawn Probability=false
   ) {module=SpawnRate001}
  ]
  particleSpawnScript [
   /SetVariables(Particles.VisibilityTag=&SelectIntFromArray(Direct Array Index=&SpawnGroup(), Int Selection Array=DataInterface() {propertyValues {IntData [0,1]}})) {isSetParameters=true}
   /SetVariables(Particles.Solid Trail Initial Color:LinearColor=<rgba 1 1 1 0.4>) {enabled=false, isSetParameters=true}
   &FN_SetSpawnGroup(Number of Spawn Groups=1)
   &InitializeParticle(
    Color=<rgba 0.7948 0.7948 0.7948 2>
    Position Offset=<0 0 30>
    Sprite Rotation Mode=::NewEnumerator2
    UsePositionOffset=false
    Write Material Random=true
   )
   &InitializeRibbon(
    Color=&MultiplyLinearColor(
     Linear Color=$System.Contrails
     Scale Factor=&MakeVector4(
      X=1
      Y=1
      Z=1
      W=&FloatFromCurve(FloatCurve=DataInterface() {propertyValues {Curve {keys [<at 0.02 0>,<at 0.4 0.7>,<at 1 0>]}}}, CurveIndex=$Emitter.NormalizedLoopAge)
     )
    )
    Lifetime=&FloatFromCurve(FloatCurve=DataInterface() {propertyValues {Curve {keys [<at 0 0.2>,<at 0.1 0.4>,<at 1.5 0>]}}}, CurveIndex=$Emitter.NormalizedLoopAge)
    Position Offset:Vector3f=<0 0 15>
    Ribbon Facing Mode:ENiagara_UnsetDirectSet=::NewEnumerator1
    Ribbon Width=&FloatFromCurve(
     FloatCurve=DataInterface() {propertyValues {Curve {keys [<at 0 1>,<at 0.3 1>,<at 1 0.725>]}}}
     CurveIndex=$Emitter.NormalizedLoopAge
     Scale Curve=12
    )
    UsePositionOffset=true
    Write Material Random=true
   )
   &SkeletalMeshLocation(
    Bone / Socket Index=$Particles.SpawnGroup
    MaskSpawnGroup=false
    Mesh Sampling Type=::NewEnumerator5
    Orientation Sampling=::NewEnumerator1
    Sampled Position Offset=<0 0 -2>
    Skeletal Mesh=$System.Skeletal Mesh
    Socket Sampling Mode=::NewEnumerator5
   )
   &CameraOffset(Camera Offset Amount=10)
  ]
  particleUpdateScript [
   &ParticleState()
   &SolveForcesAndVelocity(Clamp Velocity=false, Limit Acceleration=false)
   &ScaleColor(
    Scale Alpha=&FloatFromCurve(
     FloatCurve=DataInterface() {propertyValues {Curve {keys [
         {interpMode=RCIM_Cubic, leaveTangent=-1.0259, tangentMode=RCTM_Break, time=0.4, value=0.35}
         {arriveTangent=-0.6693, interpMode=RCIM_Cubic, leaveTangent=-0.6693, tangentMode=RCTM_User, time=0.6, value=0.06}
         {interpMode=RCIM_Cubic, tangentMode=RCTM_User, time=0.82, value=0}
        ]}}}
     CurveIndex=$Emitter.NormalizedLoopAge
     Scale Curve=&FloatFromCurve(
      FloatCurve=DataInterface() {propertyValues {Curve {keys [
          {interpMode=RCIM_Cubic, tangentMode=RCTM_User, time=0, value=0}
          {interpMode=RCIM_Cubic, tangentMode=RCTM_User, time=0.2, value=0.875}
          {tangentMode=RCTM_User, time=1, value=0}
         ]}}}
      Scale Curve=2
     )
    )
    ScaleA=true
    ScaleRGB=true
   ) {module=ScaleColor001}
   &ScaleRibbonWidth(Ribbon Width Scale=&FloatFromCurve(
     FloatCurve=DataInterface() {propertyValues {Curve {keys [
         {interpMode=RCIM_Cubic, leaveTangent=-3.5854, tangentMode=RCTM_Break, time=0, value=1.75}
         {arriveTangent=-1.2081, interpMode=RCIM_Cubic, leaveTangent=-1.2081, tangentMode=RCTM_User, time=0.5, value=0.325}
         {interpMode=RCIM_Cubic, tangentMode=RCTM_User, time=0.98, value=0}
        ]}}}
     CurveIndex=$Emitter.LoopedAge
    ))
   &ScaleColor(
    Scale Alpha=&FloatFromCurve(
     FloatCurve=DataInterface() {propertyValues {Curve {keys [
         {arriveTangent=-1.8833, interpMode=RCIM_Cubic, leaveTangent=-1.8833, tangentMode=RCTM_User, time=0, value=1}
         {interpMode=RCIM_Cubic, tangentMode=RCTM_User, time=1, value=0}
        ]}}}
     CurveIndex=$Emitter.LoopedAge
     Scale Curve=&Multiply_Float(A=1, B=&FNTimeOfDayScalar(Night Scale=1, Morning Scale=2.5, Day Scale=2.5, Evening Scale=1))
    )
    ScaleA=true
    ScaleRGB=false
   )
  ]
  rendererClasses [&NiagaraRibbonRendererProperties]
  renderers [
   {rendererClass=&NiagaraRibbonRendererProperties}
   {rendererClass=&NiagaraRibbonRendererProperties, rendererIndex=1}
  ]
 }
 Steam {
  simTarget=GPUComputeSim
  emitterUpdateScript [
   &EmitterState()
   &SpawnRate(
    SpawnRate=&FloatFromCurve(
     FloatCurve=DataInterface() {propertyValues {Curve {keys [
         <at 0 0 RCIM_Cubic>
         {arriveTangent=29.0145, interpMode=RCIM_Cubic, leaveTangent=28.2572, tangentMode=RCTM_Break, time=0.18, value=2}
         {arriveTangent=293.0562, interpMode=RCIM_Cubic, leaveTangent=128.8477, tangentMode=RCTM_Break, time=0.45, value=30}
         {arriveTangent=-155.5706, interpMode=RCIM_Cubic, leaveTangent=-155.5706, tangentMode=RCTM_User, time=0.74, value=30}
         {interpMode=RCIM_Cubic, tangentMode=RCTM_User, time=0.87, value=15}
         {interpMode=RCIM_Cubic, leaveTangent=-14.3005, tangentMode=RCTM_Break, time=1, value=15}
         {interpMode=RCIM_Cubic, tangentMode=RCTM_User, time=2, value=0}
        ]}}}
     CurveIndex=$Emitter.NormalizedLoopAge
    )
    Use Spawn Probability=true
   )
  ]
  particleSpawnScript [
   &InitializeParticle(
    Color=<rgba 1 1 1 0.14>
    Lifetime Max=1.25
    Lifetime Min=1.1
    Lifetime Mode=::NewEnumerator1
    Sprite Rotation Mode=::NewEnumerator2
    Sprite Size Mode=::NewEnumerator1
    Uniform Sprite Size Max=75
    Uniform Sprite Size Min=55
    UsePositionOffset=false
    Write Material Random=true
   )
   &CameraOffset(Camera Offset Amount=5)
   &AddVelocity(
    Rotation Coordinate Space=::World
    Velocity=&RandomRangeVector(Minimum=<0 0 0.2>, Maximum=<0 0 0.4>)
    Velocity Speed Scale=&FloatFromCurve(
     FloatCurve=DataInterface() {propertyValues {Curve {keys [{interpMode=RCIM_Cubic, tangentMode=RCTM_User, time=0, value=0.5347}, {interpMode=RCIM_Cubic, tangentMode=RCTM_User, time=1, value=1}]}}}
     CurveIndex=$Emitter.NormalizedLoopAge
     Scale Curve=60
    )
   ) {enabled=false}
   &SkeletalMeshLocation(
    MaskSpawnGroup=false
    Mesh Sampling Type=::NewEnumerator5
    Sampled Position Offset=<0 0 -6>
    Skeletal Mesh=$System.Skeletal Mesh
    Socket Sampling Mode=::NewEnumerator5
   ) {enabled=false}
   &AddVelocity(Velocity=&VectorFromCurve(
     VectorCurve=DataInterface() {propertyValues {
       XCurve {keys [<at 0 0>,<at 0.9 0>,<at 1 0>]}
       YCurve {keys [<at 0 0>,<at 0.9 0>,<at 1 0>]}
       ZCurve {keys [
         <at 0 8>
         <at 0 8>
         {arriveTangent=14.4613, interpMode=RCIM_Cubic, leaveTangent=14.4613, tangentMode=RCTM_User, time=2.5, value=16}
         {arriveTangent=24.7114, interpMode=RCIM_Cubic, leaveTangent=24.7114, tangentMode=RCTM_User, time=3, value=30}
         {interpMode=RCIM_Cubic, tangentMode=RCTM_User, time=4.9, value=55}
        ]}
      }}
     CurveIndex=$Engine.System.Age
    )) {module=AddVelocity001}
   &ShapeLocation(
    Hemisphere Distribution=<1 0.5>
    MaskSpawnGroup=false
    Offset=&VectorFromCurve(VectorCurve=DataInterface() {propertyValues {XCurve {keys [<at 0 0>,<at 1 0>]}, YCurve {keys [<at 0 0>,<at 1 0>]}, ZCurve {keys [<at 0 0>,<at 0.1 10>,<at 0.5 10>,<at 0.6 30>]}}})
    Offset Coordinate Space=::World
    Offset Mode=::NewEnumerator0
    Override Seed=false
    Sphere Radius=5
    UseOwnerScale=true
   )
  ]
  particleUpdateScript [
   &ParticleState()
   &SpriteRotationRate(Rotation Rate=&RandomRangeFloat(Minimum=-50, Maximum=50))
   &ScaleSpriteSize(
    Scale Sprite Size Mode=::NewEnumerator0
    Uniform Scale Factor=&FloatFromCurve(
     FloatCurve=DataInterface() {propertyValues {Curve {
        defaultValue=0.8
        keys [
         {arriveTangent=1.3252, interpMode=RCIM_Cubic, leaveTangent=1.3252, tangentMode=RCTM_User, time=0, value=0}
         {arriveTangent=1.3536, interpMode=RCIM_Cubic, leaveTangent=1.5415, tangentMode=RCTM_Break, time=0.6, value=1}
         {arriveTangent=0.0224, interpMode=RCIM_Cubic, leaveTangent=0.0224, tangentMode=RCTM_User, time=1, value=1.25}
        ]
       }}}
     Scale Curve=0.45
    )
   )
   &Drag(Drag=0.4, Use Linear Drag=true, Use Rotational Drag=true)
   &SubUVAnimation(
    End Frame Range Override=30
    "Sprite Renderer "=DataInterface() {propertyValues {SpriteRenderer=&self:Steam_0.NiagaraSpriteRendererProperties_0}}
    Start Frame Range Override=&RandomRangeInt(Minimum=0, Maximum=5)
    UseEndFrame=true
    UseStartFrame=true
   )
   &ScaleColor(
    Scale RGB=&MakeVectorFromLinearColorRGB(Color=&ColorFromCurve(
      ColorCurve=DataInterface() {propertyValues {
        AlphaCurve {keys []}
        BlueCurve {keys [
          <at 0.008 0.1289>
          {arriveTangent=0.6421, interpMode=RCIM_Cubic, leaveTangent=0.6421, time=0.6663, value=0.169}
          {arriveTangent=0.6421, interpMode=RCIM_Cubic, leaveTangent=0.6421, time=0.9436, value=0.4577}
          <at 1 0.766>
         ]}
        GreenCurve {keys [
          <at 0.008 0.1369>
          {arriveTangent=0.6754, interpMode=RCIM_Cubic, leaveTangent=0.6754, time=0.6663, value=0.1751}
          {arriveTangent=0.6754, interpMode=RCIM_Cubic, leaveTangent=0.6754, time=0.9436, value=0.4822}
          <at 1 0.8069>
         ]}
        RedCurve {keys [
          <at 0.008 0.1383>
          {arriveTangent=0.6796, interpMode=RCIM_Cubic, leaveTangent=0.6796, time=0.6663, value=0.1753}
          {arriveTangent=0.6796, interpMode=RCIM_Cubic, leaveTangent=0.6796, time=0.9436, value=0.4855}
          <at 1 0.8125>
         ]}
       }}
      CurveIndex=$Engine.System.Age
     ))
    ScaleA=true
    ScaleRGB=true
   ) {module=ScaleColor001}
   &ScaleColor(
    Scale Alpha=&Multiply_Float(
     A=&RandomRangeFloat(Minimum=0.5, Maximum=1)
     B=&FloatFromCurve(
      FloatCurve=DataInterface() {propertyValues {Curve {keys [
          {interpMode=RCIM_Cubic, tangentMode=RCTM_User, time=0, value=0}
          {arriveTangent=0.579, interpMode=RCIM_Cubic, leaveTangent=0.579, tangentMode=RCTM_User, time=0.15, value=0.75}
          {arriveTangent=-1.4979, interpMode=RCIM_Cubic, leaveTangent=-1.4979, tangentMode=RCTM_User, time=0.35, value=0.35}
          {tangentMode=RCTM_User, time=1, value=0}
         ]}}}
      Scale Curve=&Multiply_Float(
       A=&FN_FrontEndFloatModifier(Gameplay=1.5, Front End=1)
       B=&FNTimeOfDayScalar(Night Scale=0.5, Morning Scale=0.6, Day Scale=0.6, Evening Scale=0.8)
      )
     )
    )
    Scale RGB=&MakeVectorFromLinearColorRGB(Color=&MakeCustomLinearColorFromBool(
      Bool=$User.bIsFrontEndPreview
      True Color=<rgba 0.7865 0.8826 1 1>
      False Color=&FNTimeOfDayColor(Night Color=<rgba 0.5888 0.6049 0.8156 1>, Morning Color=<rgba 1 1 1 1>, Day Color=<rgba 1 1 1 1>, Evening Color=<rgba 1 1 1 1>)
     ))
    ScaleA=true
    ScaleRGB=true
   )
   &InheritSourceMovement(Applied Position Scale=<0.3 0.3 0.3>, ApplyForce=false, ApplyPosition=true, ApplyPositionFalloff=false)
   &SolveForcesAndVelocity(Clamp Velocity=false, Limit Acceleration=false)
   &ScaleColor(
    Scale Alpha=&FloatFromCurve(
     FloatCurve=DataInterface() {propertyValues {Curve {keys [
         {interpMode=RCIM_Cubic, leaveTangentWeight=1.3362, tangentMode=RCTM_User, time=0, value=1}
         {arriveTangentWeight=1.3167, interpMode=RCIM_Cubic, leaveTangentWeight=0.0167, tangentMode=RCTM_User, time=3, value=1}
         {arriveTangent=-1.7575, interpMode=RCIM_Cubic, leaveTangent=-1.7575, tangentMode=RCTM_User, time=4, value=0}
        ]}}}
     CurveIndex=$Engine.System.Age
    )
    ScaleA=true
    ScaleRGB=false
   ) {module=ScaleColor002}
  ]
  rendererClasses [&NiagaraSpriteRendererProperties]
  renderers [{rendererClass=&NiagaraSpriteRendererProperties}]
 }
}
summary {
 systemName=NS_Emote_ForwardLakeDire_MotionTrailSteam
 emitters [
  {emitterName=Steam, simTarget=GPUComputeSim, rendererClasses [&NiagaraSpriteRendererProperties]}
  {emitterName=MotionTrail001, rendererClasses [&NiagaraRibbonRendererProperties]}
 ]
 userVariables [
  User.bIsFrontEnd {type=NiagaraBool}
  User.bIsFrontEndPreview {type=NiagaraBool}
  User.NewVector3f {type=Vector3f, description="", defaultValue {struct=Vector3f, value=<10 0 0>}}
  User.RibbonIDOffset {type=NiagaraInt32}
  User.SkeletalMesh {
   type=NiagaraDataInterfaceSkeletalMesh
   defaultValue {
    struct=NiagaraExt_VariableValue_DataInterface
    value {dataInterface=&self:NiagaraDataInterfaceSkeletalMesh_0}
   }
  }
 ]
}
